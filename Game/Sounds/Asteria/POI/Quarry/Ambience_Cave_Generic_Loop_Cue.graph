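# Sound cue for a generic looping cave ambience in the Quarry POI, reusing the Cattus Cave loop recordings from Artemis.
# A quality level node at the root selects the looping quad wave for the first two levels and the looping stereo wave for the last two.
# It uses ambient attenuation and an indoor ambient sound class, with concurrency overrides set to one instance that blocks new ones.
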SoundCue /Game/Sounds/Asteria/POI/Quarry/Ambience_Cave_Generic_Loop_Cue
root=QualityLevel_0
nodes [
 {id=QualityLevel_0, children [WavePlayer_2,WavePlayer_2,WavePlayer_3,WavePlayer_3]}
 {id=WavePlayer_2, properties {bLooping=true, soundWaveAssetPtr=/Game/Sounds/Artemis/POI/Cattus_Cave/Ambience_Cattus_Cave_Loop_Quad}}
 {id=WavePlayer_3, properties {bLooping=true, soundWaveAssetPtr=/Game/Sounds/Artemis/POI/Cattus_Cave/Ambience_Cattus_Cave_Loop_Stereo}}
]
properties {
 attenuationSettings=&Fort_Ambient_Athena_Att
 maxDistance=2097152
 soundClassObject=&Fort_Ambient_Indoors_Athena
 volumeMultiplier=1
 concurrencyOverrides {maxCount=1, resolutionRule=PreventNew}
}
