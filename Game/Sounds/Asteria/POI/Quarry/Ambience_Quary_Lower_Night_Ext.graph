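# Looping outdoor ambience cue for the lower Quarry POI at night, reusing IndustrialTown night recordings.
# A quality-level node sends its first slot to a quad loop and the other three slots to a shared stereo loop branch.
# The quad branch nests two indefinitely looping envelopers over a slight random pitch modulator.
# The stereo branch randomizes pitch and volume above a looping enveloper that slowly varies both.
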SoundCue /Game/Sounds/Asteria/POI/Quarry/Ambience_Quary_Lower_Night_Ext
root=QualityLevel_0
nodes [
 {
  id=Enveloper_0
  children [Enveloper_3]
  properties {
   bLoop=true
   bLoopIndefinitely=true
   loopEnd=100
   pitchCurve {editorCurveData {keys [<at 0 0.8>]}}
   volumeCurve {editorCurveData {keys [<at 0 1 RCIM_Cubic>,<at 50 0.9 RCIM_Cubic>,<at 100 1 RCIM_Cubic>]}}
  }
 }
 {
  id=Enveloper_1
  children [WavePlayer_13]
  properties {
   bLoop=true
   bLoopIndefinitely=true
   loopEnd=30
   pitchCurve {editorCurveData {keys [{interpMode=RCIM_Cubic, time=0}, <at 15.9468 0.9181 RCIM_Cubic>, {interpMode=RCIM_Cubic, time=30}]}}
   volumeCurve {editorCurveData {keys [<at 0 1 RCIM_Cubic>,<at 16.1421 0.921 RCIM_Cubic>,<at 30 1 RCIM_Cubic>]}}
  }
 }
 {
  id=Enveloper_3
  children [Modulator_11]
  properties {
   bLoop=true
   bLoopIndefinitely=true
   loopEnd=30
   pitchCurve {editorCurveData {keys [{interpMode=RCIM_Cubic, time=0}, <at 15.9468 0.9181 RCIM_Cubic>, {interpMode=RCIM_Cubic, time=30}]}}
   volumeCurve {editorCurveData {keys [<at 0 1>]}}
  }
 }
 {id=Modulator_11, children [WavePlayer_1], properties {pitchMax=1, pitchMin=0.97}}
 {id=Modulator_2, children [Enveloper_1], properties {pitchMax=1.1, pitchMin=0.9, volumeMax=1.1, volumeMin=0.9}}
 {id=QualityLevel_0, children [Enveloper_0,Modulator_2,Modulator_2,Modulator_2]}
 {id=WavePlayer_1, properties {bLooping=true, soundWaveAssetPtr=/Game/Sounds/Asteria/POI/IndustrialTown/Ambience_Industrial_Night_01_Quad}}
 {id=WavePlayer_13, properties {bLooping=true, soundWaveAssetPtr=/Game/Sounds/Asteria/POI/IndustrialTown/Ambience_Industrial_Night_03_Stereo}}
]
properties {maxDistance=2097152, soundClassObject=&Fort_Ambient_Outdoors_Athena, volumeMultiplier=0.501}
