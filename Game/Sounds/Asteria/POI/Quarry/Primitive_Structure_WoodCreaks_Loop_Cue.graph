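# Looping ambient Sound Cue that plays intermittent wooden structure creaks for the Quarry POI.
# A Looping node repeats a Modulator with slightly lowered random pitch, feeding a random Delay of several seconds before an equal-weight Random pick among ten creak waves.
# The cue plays quiet and pitched down in the outdoor ambient sound class, using square map attenuation and a shared structure concurrency setting.
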SoundCue /Game/Sounds/Asteria/POI/Quarry/Primitive_Structure_WoodCreaks_Loop_Cue
root=Looping_0
nodes [
 {id=Delay_0, children [Random_0], properties {delayMax=12, delayMin=3}}
 {id=Looping_0, children [Modulator_0]}
 {id=Modulator_0, children [Delay_0], properties {pitchMax=1, pitchMin=0.8}}
 {
  id=Random_0
  children [WavePlayer_0,WavePlayer_1,WavePlayer_2,WavePlayer_3,WavePlayer_4,WavePlayer_5,WavePlayer_6,WavePlayer_7,WavePlayer_8,WavePlayer_9]
  properties {weights [1,1,1,1,1,1,1,1,1,1]}
 }
 {id=WavePlayer_0, properties {soundWaveAssetPtr=/Game/Sounds/Fort_Ambient_NonLoop/Primitive/Primitive_Structure_WoodCreaks_10}}
 {id=WavePlayer_1, properties {soundWaveAssetPtr=/Game/Sounds/Fort_Ambient_NonLoop/Primitive/Primitive_Structure_WoodCreaks_01}}
 {id=WavePlayer_2, properties {soundWaveAssetPtr=/Game/Sounds/Fort_Ambient_NonLoop/Primitive/Primitive_Structure_WoodCreaks_02}}
 {id=WavePlayer_3, properties {soundWaveAssetPtr=/Game/Sounds/Fort_Ambient_NonLoop/Primitive/Primitive_Structure_WoodCreaks_03}}
 {id=WavePlayer_4, properties {soundWaveAssetPtr=/Game/Sounds/Fort_Ambient_NonLoop/Primitive/Primitive_Structure_WoodCreaks_04}}
 {id=WavePlayer_5, properties {soundWaveAssetPtr=/Game/Sounds/Fort_Ambient_NonLoop/Primitive/Primitive_Structure_WoodCreaks_05}}
 {id=WavePlayer_6, properties {soundWaveAssetPtr=/Game/Sounds/Fort_Ambient_NonLoop/Primitive/Primitive_Structure_WoodCreaks_06}}
 {id=WavePlayer_7, properties {soundWaveAssetPtr=/Game/Sounds/Fort_Ambient_NonLoop/Primitive/Primitive_Structure_WoodCreaks_07}}
 {id=WavePlayer_8, properties {soundWaveAssetPtr=/Game/Sounds/Fort_Ambient_NonLoop/Primitive/Primitive_Structure_WoodCreaks_08}}
 {id=WavePlayer_9, properties {soundWaveAssetPtr=/Game/Sounds/Fort_Ambient_NonLoop/Primitive/Primitive_Structure_WoodCreaks_09}}
]
properties {
 attenuationSettings=/Game/Sounds/Attenuation/Fort_Map_3x3_Square
 maxDistance=1780
 pitchMultiplier=0.8
 soundClassObject=&Fort_Ambient_Outdoors_Athena
 volumeMultiplier=0.25
 concurrencySet [/Game/Sounds/Fort_Ambient_NonLoop/Primitive/Primitive_Structure_GLobal_Concurrency]
}
