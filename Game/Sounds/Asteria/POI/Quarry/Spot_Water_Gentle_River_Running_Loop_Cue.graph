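# Sound cue for a looping gentle river spot ambience at the Asteria Quarry POI, routed to the Fort water ambient sound class.
# A looping close stream wave feeds an Enveloper, then an Attenuation node using the shared Apollo close-river settings.
# The Enveloper's volume curve swells and dips over twenty seconds to vary the stream's level, while its pitch curve holds a single key.
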SoundCue /Game/Sounds/Asteria/POI/Quarry/Spot_Water_Gentle_River_Running_Loop_Cue
root=Attenuation_1
nodes [
 {
  id=Attenuation_1
  children [Enveloper_0]
  properties {attenuationSettings=/Game/Athena/Sounds/Water/Attenuations/Apollo_Water_River_Close_Att}
 }
 {
  id=Enveloper_0
  children [WavePlayer_4]
  properties {
   pitchCurve {editorCurveData {keys [{time=0}]}}
   volumeCurve {editorCurveData {keys [
      <at 0 1 RCIM_Cubic>
      {arriveTangent=-0.0013, interpMode=RCIM_Cubic, leaveTangent=-0.0013, tangentMode=RCTM_User, time=4.0805, value=0.6}
      <at 4.0805 0.7528>
      {arriveTangent=0.0393, interpMode=RCIM_Cubic, leaveTangent=0.0393, tangentMode=RCTM_User, time=6.3837, value=0.9043}
      {arriveTangent=-0.012, interpMode=RCIM_Cubic, leaveTangent=-0.012, time=10, value=0.5}
      {arriveTangent=0.0274, interpMode=RCIM_Cubic, leaveTangent=0.0274, tangentMode=RCTM_User, time=12.4126, value=0.832}
      {arriveTangent=-0.0008, interpMode=RCIM_Cubic, leaveTangent=-0.0008, tangentMode=RCTM_User, time=15.9351, value=0.6}
      <at 20 1>
     ]}}
  }
 }
 {id=WavePlayer_4, properties {bLooping=true, soundWaveAssetPtr=/Game/Athena/Sounds/Water/Ambient/Amb_Water_Stream_Med_Close_Loop_02}}
]
properties {
 maxDistance=1600
 soundClassObject=/Game/Sounds/FortSoundClassesAndModes/Fort_Water_Ambient
 volumeMultiplier=0.6
 attenuationOverrides {
  bAttenuateWithLPF=true
  falloffDistance=6000
  lPFFrequencyAtMax=500
  nonSpatializedRadiusStart=700
  stereoSpread=1000
  attenuationShapeExtents {x=925}
 }
 concurrencyOverrides {maxCount=1, resolutionRule=PreventNew}
}
